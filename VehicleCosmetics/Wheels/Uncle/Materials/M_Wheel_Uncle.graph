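# M_Wheel_Uncle is the masked, two-sided material for the Uncle cosmetic vehicle wheel, built on the shared MF_WheelParent_Simple_Alpha function.
# It overrides emissive with a mask-limited, color-tinted glow from two swirl texture samples panning at different speeds.
# It also overrides the normal, with a static switch for tiled asphalt detail on tire rubber.
# The resulting attributes pass through vehicle damage, then de-rez transition functions.
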Material /VehicleCosmetics/Wheels/Uncle/Materials/M_Wheel_Uncle
graph {
 properties {
  bAutomaticallySetUsageInEditor=false
  bUseMaterialAttributes=true
  bUsedWithMeshDeformer=true
  bUsedWithSkeletalMesh=true
  blendMode=BLEND_Masked
  pixelDepthOffsetMode=PDOM_Legacy
  twoSided=true
  usedShadingModels=Default Lit
 }
 nodes [
  n1 = MaterialFunctionCall() {materialFunction=/VehicleCosmetics/Wheels/Mage/Materials/MF_WheelParent_Simple_Alpha}
  n2 = GetMaterialAttributes(None=@n1.MaterialAttributes) {attributeGetTypes [#1, #2, #3]}
  n3 = NamedRerouteUsage()
  n4 = NamedRerouteUsage()
  MaterialFunctionCall(In=MaterialFunctionCall(Material=SetMaterialAttributes!(MaterialAttributes=@n2.MaterialAttributes, Normal=@n3, Emissive Color=@n4) {attributeSetTypes [#1, #4]}) {materialFunction=&MF_BaseVehicle_Damage}.Result) {materialFunction=&MF_VehicleDeRez_Transition}
  n5 = LinearInterpolate(A=NamedRerouteUsage(), B=NamedRerouteUsage(), Alpha=NamedRerouteUsage())
  n6 = VectorParameter() {group=Emisssive, parameterName=CustomColor, defaultValue=<rgba 0.2441 1 0.137 1>}
  n7 = TextureCoordinate() {uTiling=4, vTiling=0.5}
  n8 = TextureCoordinate() {uTiling=2, vTiling=0.5}
  n9 = Multiply(
   A=TextureSample(UVs=Panner(Coordinate=@n7) {speedX=-0.05}) {texture=/VehicleCosmetics/Wheels/Uncle/Textures/Swirls02_D}.RGB
   B=TextureSample(UVs=Panner(Coordinate=@n8) {speedX=-0.11}) {texture=&Swirls02_D}.RGB
  )
  n10 = Constant() {r=1.5}
  n11 = LinearInterpolate(
   A=Multiply(A=Constant() {r=0.5}, B=@n6.RGB)
   B=Multiply(A=@n6.RGB, B=ScalarParameter() {defaultValue=2, group=Emisssive, parameterName=GlowAmount})
   Alpha=Multiply(A=Power(Base=@n9, Exp=@n10), B=Constant() {r=20})
  )
  NamedRerouteDeclaration(None=Clamp(None=Multiply(A=@n5, B=@n11))) {name=Emissive, nodeColor=<rgba 0.5703 1 0 1>}
  NamedRerouteDeclaration(None=VertexColor().R) {name=VC_R, nodeColor=<rgba 1 0 0.8906 1>}
  NamedRerouteDeclaration(None=ComponentMask(None=@n1.RimRGBAMask) {b=true}) {name=RIM_B, nodeColor=<rgba 0 0.5078 1 1>}
  NamedRerouteDeclaration(None=@n2.Normal) {name=Normal IN, nodeColor=<rgba 1 0.4219 0 1>}
  n12 = NamedRerouteUsage()
  n13 = Multiply(A=TextureCoordinate(), B=ScalarParameter() {defaultValue=8, group=TireRubber, parameterName=TextureDetailNormalTiling})
  n14 = Add(
   A=@n12
   B=Multiply(
    A=TextureSample(UVs=@n13) {samplerType=SAMPLERTYPE_Normal, texture=/VehicleCosmetics/Wheels/BlueSteelSE_Infinity/Textures/Asphalt_N}.RGB
    B=NamedRerouteUsage()
   )
  )
  n15 = NamedRerouteUsage()
  NamedRerouteDeclaration(None=StaticSwitchParameter(True=LinearInterpolate(A=@n14, B=@n12, Alpha=@n15), False=@n12) {group=TireRubber, parameterName=TireRubberExist}) {name=NormalRubber, nodeColor=<rgba 0.1016 0 1 1>}
  NamedRerouteDeclaration(None=ComponentMask(None=@n1.TireRGBAMask) {b=true}) {name=TIRE_B, nodeColor=<rgba 0.8281 1 0 1>}
  NamedRerouteDeclaration(None=ComponentMask(None=@n1.TireRGBAMask) {g=true}) {name=TIRE_G, nodeColor=<rgba 0.125 1 0 1>}
 ]
}
